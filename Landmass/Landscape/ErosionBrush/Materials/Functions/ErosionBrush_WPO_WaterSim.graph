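# Material function shading Landmass erosion brush terrain with simulated water, outputting base color, roughness, emissive, normals and world position offsets.
# HeightRT supplies terrain height and water amount, while NormalRT packs terrain and water normals.
# Water shading combines exponential absorption and scattering, Fresnel sky reflection, and refracted UV offsets with spiral-blurred normals.
# WPO displaces vertically for terrain alone or terrain plus water, and a red mask flags out-of-range Z.
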MaterialFunction /Landmass/Landscape/ErosionBrush/Materials/Functions/ErosionBrush_WPO_WaterSim
graph {nodes [
  FunctionOutput(None=Multiply(A=FunctionInput() {bUsePreviewValueAsDefault=true, inputName=BaseColor, previewValue=<1 1 1 1>}, B=NamedRerouteUsage())) {outputName=BaseColor}
  FunctionOutput(None=LinearInterpolate(
    A=ScalarParameter() {defaultValue=1, parameterName=LandR}
    B=ScalarParameter() {defaultValue=1, parameterName=WaterR}
    Alpha=NamedRerouteUsage()
   )) {outputName=Roughness, sortPriority=1}
  n1 = NamedRerouteUsage()
  n2 = NamedRerouteUsage()
  FunctionOutput(None=Add(A=NamedRerouteUsage(), B=Add(A=@n1, B=@n2))) {outputName=Emissive, sortPriority=2}
  n3 = NamedRerouteUsage()
  FunctionOutput(None=LinearInterpolate(A=NamedRerouteUsage(), B=NamedRerouteUsage(), Alpha=Saturate(None=@n3))) {outputName=Water and Terrain Normal, sortPriority=4}
  FunctionOutput(None=NamedRerouteUsage()) {outputName=Terrain Normal, sortPriority=3}
  n4 = NamedRerouteUsage()
  n5 = MaterialFunctionCall() {materialFunction=/Landmass/Landscape/BlueprintBrushes/MF/Landmass_TerrainMesh_Coords}
  n6 = NamedRerouteDeclaration(None=Multiply(
    A=TextureSampleParameter2D(UVs=@n5.Landscape UVs) {parameterName=HeightRT, samplerSource=SSM_Wrap_WorldGroupSettings, texture=&Black}.R
    B=CollectionParameter() {collection=/Landmass/Landscape/BlueprintBrushes/MPC/MPC_Erosion, parameterName=ZScale}
   )) {name=Landscape Z, nodeColor=<rgba 0 1 0.0156 1>}
  FunctionOutput(None=Multiply(A=Add(A=@n6, B=@n4), B=Constant3Vector() {constant=<rgba 0 0 1 0>})) {outputName=WPO Terrain + Water, sortPriority=5}
  FunctionOutput(None=Multiply(A=@n6, B=Constant3Vector() {constant=<rgba 0 0 1 0>})) {outputName=WPO Terrain, sortPriority=6}
  MaterialFunctionCall() {materialFunction=&/ErosionBrush_ParticleCoords}
  MaterialFunctionCall() {materialFunction=&ErosionBrush_ParticleCoords}
  MaterialFunctionCall() {materialFunction=&ErosionBrush_ParticleCoords}
  MaterialFunctionCall() {materialFunction=&ErosionBrush_ParticleCoords}
  n7 = Divide(A=NamedRerouteUsage(), B=CollectionParameter() {collection=&MPC_Erosion, parameterName=ZScale})
  n8 = Saturate(None=Ceil(None=Subtract(A=@n7) {constB=32767}))
  n9 = Multiply(A=@n7) {constB=-1}
  n10 = Saturate(None=Ceil(None=Subtract(A=@n9) {constB=32767}))
  NamedRerouteDeclaration(None=Multiply(A=Constant3Vector() {constant=<rgba 1 0 0 0>}, B=Max(A=@n8, B=@n10))) {name=Clamping Z Range Error, nodeColor=<rgba 1 0 0.0398 1>}
  MaterialExpressionDivide_4 = Divide(A=NamedRerouteUsage()) {constB=100}
  MaterialExpressionScalarParameter_23 = ScalarParameter() {defaultValue=0.02, parameterName=Scatter Density}
  n11 = Custom!(t=@MaterialExpressionDivide_4, c=@MaterialExpressionScalarParameter_23) {
   code="return exp(-t * c);"
   description=Exponential Density
   inputs [
    {inputName=t, input {expression=&self:MaterialExpressionDivide_4}}
    {
     inputName=c
     input {expression=&self:MaterialExpressionScalarParameter_23}
    }
   ]
  }
  n12 = OneMinus(None=NamedRerouteDeclaration(None=@n11) {name=Scattering Transmittance, nodeColor=<rgba 0 1 0.9766 1>})
  n13 = VectorParameter() {parameterName=Scattering Color, defaultValue=<rgba 0.1244 0.9701 1 0.1>}
  n14 = Multiply(A=@n13.RGB, B=@n13.RGB)
  n15 = DotProduct(A=NamedRerouteUsage(), B=Multiply(A=AtmosphericLightVector()) {constB=-1})
  NamedRerouteDeclaration(None=Multiply(A=Multiply(A=@n12, B=@n14), B=ConstantBiasScale(None=@n15))) {name=Scattering, nodeColor=<rgba 0 1 0.2734 1>}
  n16 = NamedRerouteUsage()
  n17 = ScalarParameter() {defaultValue=0.02, parameterName=Water Edge Sharpness}
  NamedRerouteDeclaration(None=Saturate(None=Multiply(A=@n16, B=@n17))) {name=Water Mask, nodeColor=<rgba 0.0126 0.2876 1 1>}
  n18 = ScalarParameter() {defaultValue=3, parameterName=FresnelExp}
  n19 = NamedRerouteUsage()
  n20 = Multiply(
   A=NamedRerouteDeclaration(None=Fresnel(ExponentIn=@n18, Normal=@n19)) {name=Fresnel Mask, nodeColor=<rgba 0 1 1 1>}
   B=TextureSample(UVs=MaterialFunctionCall!(Normal=@n19, CameraVector) {materialFunction=/Engine/Functions/Engine_MaterialFunctions02/Utility/CustomReflectionVector}.Reflection Vector) {samplerType=SAMPLERTYPE_LinearColor, texture=/Landmass/PreviewContent/Sky/Tex/Sky54a2048}.RGB
  )
  n21 = Multiply(A=ScalarParameter() {defaultValue=0.5, parameterName=Cubemap Brightness}, B=AtmosphericLightColor())
  NamedRerouteDeclaration(None=Multiply(A=Multiply(A=@n20, B=@n21), B=NamedRerouteUsage())) {name=Reflection, nodeColor=<rgba 0.0781 1 0 1>}
  n22 = NamedRerouteUsage()
  n23 = NamedRerouteUsage()
  MaterialExpressionNamedRerouteDeclaration_10 = NamedRerouteDeclaration(None=MaterialFunctionCall(Vector 2=Multiply(A=@n22, B=@n23)) {materialFunction=/Engine/Functions/Engine_MaterialFunctions02/Utility/VectorLength}.V2 Length) {name=Underwater Ray Length, nodeColor=<rgba 0 1 1 1>}
  MaterialExpressionDivide_2 = Divide(
   A=ScalarParameter() {defaultValue=8, parameterName=Absorption Density Scale}
   B=VectorParameter() {parameterName=Absorption, defaultValue=<rgba 250 5000 227000 0>}.RGB
  )
  n24 = Custom(t=@MaterialExpressionNamedRerouteDeclaration_10, c=@MaterialExpressionDivide_2) {
   code="return exp(-t * c);"
   description=Exponential Density
   inputs [
    {
     inputName=t
     input {expression=&self:MaterialExpressionNamedRerouteDeclaration_10}
    }
    {inputName=c, input {expression=&self:MaterialExpressionDivide_2}}
   ]
  }
  n25 = NamedRerouteUsage()
  n26 = OneMinus(None=NamedRerouteUsage())
  n27 = NamedRerouteUsage()
  NamedRerouteDeclaration(None=Multiply(A=Multiply(A=@n24, B=@n25) {desc=* Scattering Opacity}, B=LinearInterpolate(B=@n26, Alpha=@n27) {constA=1})) {name=Water Absorption, nodeColor=<rgba 0 0.6693 1 1>}
  n28 = TextureSampleParameter2D(UVs=MaterialFunctionCall() {materialFunction=&Landmass_TerrainMesh_Coords}.Landscape UVs) {parameterName=HeightRT, samplerSource=SSM_Wrap_WorldGroupSettings, texture=&Black}
  NamedRerouteDeclaration(None=Multiply(
    A=CollectionParameter() {collection=&MPC_Erosion, parameterName=ShowWater}
    B=Max(A=@n28.G, B=@n28.G)
   )) {name=Water Amount, nodeColor=<rgba 0.0205 0.4638 1 1>}
  n29 = TextureObjectParameter() {parameterName=NormalRT, texture=&Black}
  n30 = Add(
   A=MaterialFunctionCall() {materialFunction=&Landmass_TerrainMesh_Coords}.Landscape UVs
   B=NamedRerouteUsage()
  )
  n31 = NamedRerouteUsage()
  n32 = Add(
   A=ScalarParameter() {defaultValue=0.0001, parameterName=mindist}
   B=Multiply(A=ScalarParameter() {defaultValue=0.0003, parameterName=BlurDist}, B=Divide(A=@n31) {constB=100})
  )
  n33 = Min(A=ScalarParameter() {defaultValue=3, parameterName=BlurSteps}) {constB=8}
  n34 = ConstantBiasScale(None=ComponentMask(None=MaterialFunctionCall!(TextureObject=@n29, UV=@n30, Distance=@n32, Distance Steps=@n33, Radial Steps=@n33, Radial Offset, Kernel Power) {materialFunction=/Engine/Functions/Engine_MaterialFunctions01/Texturing/SpiralBlur-Texture}.Result) {g=true, r=true}) {bias=-0.5, scale=2}
  n35 = AppendVector(A=Constant(), B=ScalarParameter() {defaultValue=0.01, parameterName=NormalBias_1})
  NamedRerouteDeclaration(None=DeriveNormalZ(InXY=Add(A=@n34, B=@n35))) {name=Blurred Normal, nodeColor=<rgba 0 0.1094 1 1>}
  n36 = ScalarParameter() {parameterName=NormalBias}
  n37 = AppendVector(A=Divide(A=@n36) {constB=1.618}, B=@n36)
  n38 = TextureSampleParameter2D(UVs=MaterialFunctionCall() {materialFunction=&Landmass_TerrainMesh_Coords}.Landscape UVs) {parameterName=NormalRT, samplerSource=SSM_Wrap_WorldGroupSettings, texture=&Black}
  n39 = ConstantBiasScale(None=AppendVector(A=@n38.RGB, B=@n38.RGB)) {bias=-0.5, scale=2}
  n40 = ComponentMask(None=@n39) {a=true, b=true}
  NamedRerouteDeclaration(None=DeriveNormalZ(InXY=Add(A=@n37, B=@n40))) {name=Water Normal, nodeColor=<rgba 0 0.2031 1 1>}
  NamedRerouteDeclaration(None=DeriveNormalZ(InXY=ComponentMask(None=@n39) {g=true, r=true})) {name=Terrain Normal, nodeColor=<rgba 0.6875 1 0 1>}
  n41 = NamedRerouteDeclaration(None=MaterialFunctionCall!(
    Refractive Index 1
    Refractive Index 2=ScalarParameter() {defaultValue=1.33, parameterName=n2}
    Surface Normal=NamedRerouteUsage()
    Camera Vector=CameraVectorWS()
   ) {materialFunction=/Landmass/Landscape/OldPrototype_BP/Mats/MF/RefractionWorldVector}.Refracted Vector) {name=Refracted RayDir, nodeColor=<rgba 1 0 0.5156 1>}
  n42 = Multiply(A=Divide(A=ComponentMask(None=@n41) {g=true, r=true}, B=ComponentMask(None=@n41) {b=true})) {constB=-1}
  NamedRerouteDeclaration(None=@n42) {name=Refraction Extrapolation, nodeColor=<rgba 0 0.1094 1 1>}
  n43 = Divide(A=NamedRerouteUsage(), B=CollectionParameter() {collection=&MPC_Erosion, parameterName=TotalSizeX})
  n44 = ScalarParameter() {defaultValue=1, parameterName=Depth}
  NamedRerouteDeclaration(None=Multiply(A=@n42, B=Multiply(A=@n43, B=@n44))) {name=Refraction Tangent Scaled, nodeColor=<rgba 1 0 0.1172 1>}
 ]}
