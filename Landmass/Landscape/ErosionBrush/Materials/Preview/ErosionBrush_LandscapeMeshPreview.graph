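# Lit preview material for the Landmass erosion brush's landscape mesh, usable on instanced static meshes.
# Two weightmaps blend a base value, a weightmap color and a desert color, feeding the water-sim WPO function, whose emissive adds to a rerouted input.
# ErosionBrush_ParticleCoords supplies landscape UVs for weightmap, height and scene-capture particle textures.
# Debug branches build a component-border grid from world position and MPC_Erosion values, a numeric readout, and a green height-plus-particles overlay.
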Material /Landmass/Landscape/ErosionBrush/Materials/Preview/ErosionBrush_LandscapeMeshPreview
graph {
 properties {bTangentSpaceNormal=false, bUsedWithInstancedStaticMeshes=true, pixelDepthOffsetMode=PDOM_Legacy, usedShadingModels=Default Lit}
 nodes [
  n1 = LinearInterpolate(
   A=ScalarParameter() {defaultValue=0.12, parameterName=Base}
   B=VectorParameter() {parameterName=Weightmap1_Color, defaultValue=<rgba 0.07 0.1719 0.0638 1>}.RGB
   Alpha=TextureSampleParameter2D(UVs=MaterialFunctionCall() {materialFunction=/Landmass/Landscape/ErosionBrush/Materials/Functions/ErosionBrush_ParticleCoords}.Landscape UVs) {parameterName=Weightmap1, texture=&Black}.R
  )
  n2 = VectorParameter() {parameterName=DesertColor, defaultValue=<rgba 0.9688 0.4848 0.2069 1>}
  n3 = TextureSampleParameter2D(UVs=MaterialFunctionCall() {materialFunction=&ErosionBrush_ParticleCoords}.Landscape UVs) {parameterName=Weightmap2, texture=&Black}
  Add(
   A=MaterialFunctionCall!(BaseColor=LinearInterpolate(A=@n1, B=@n2.RGB, Alpha=@n3.R)) {materialFunction=/Landmass/Landscape/ErosionBrush/Materials/Functions/ErosionBrush_WPO_WaterSim}.Emissive
   B=NamedRerouteUsage()
  )
  n4 = ComponentMask(None=WorldPosition().XYZ) {g=true, r=true}
  n5 = CollectionParameter() {collection=/Landmass/Landscape/BlueprintBrushes/MPC/MPC_Erosion, parameterName=ComponentRes}
  n6 = CollectionParameter() {collection=&MPC_Erosion, parameterName=GridSize}
  n7 = Multiply(A=Subtract(A=@n5), B=@n6)
  n8 = MaterialFunctionCall!(
   A=Frac(None=Divide(A=@n4, B=@n7))
   B=Constant() {r=0.5}
   Bounds=ScalarParameter() {defaultValue=0.997, parameterName=B2}
   Edge Falloff=ScalarParameter() {parameterName=ef}
  ) {materialFunction=/Engine/Functions/Engine_MaterialFunctions01/Gradient/BoxMask-2D}
  Add(
   A=Multiply(
    A=OneMinus(None=@n8.Result)
    B=CollectionParameter() {collection=&MPC_Erosion, parameterName=ShowComponents}
   )
   B=MaterialFunctionCall!(Vector2=AppendVector(A=@n5, B=@n6), MaximumNumberOfDigits, UVs, DebugTextLocation RG_UpperRight BA_LowerLeft, Component Spacing) {materialFunction=/Engine/Functions/Engine_MaterialFunctions02/Utility/DebugFloat2Values}.ColorCodedOutput
  )
  n9 = MaterialFunctionCall() {materialFunction=&ErosionBrush_ParticleCoords}
  n10 = Multiply(
   A=TextureSampleParameter2D(UVs=MaterialFunctionCall() {materialFunction=&ErosionBrush_ParticleCoords}.Landscape UVs) {parameterName=SceneCaptureRT, samplerSource=SSM_Wrap_WorldGroupSettings, texture=&Black}.RGB
   B=ScalarParameter() {parameterName=Show Particles}
  )
  Add(
   A=Multiply(
    A=TextureSampleParameter2D(UVs=@n9.Landscape UVs) {parameterName=HeightRT, samplerSource=SSM_Wrap_WorldGroupSettings, texture=&Black}.G
    B=Constant3Vector() {constant=<rgba 0 0.5 0 0>}
   )
   B=@n10
  )
  NamedRerouteDeclaration(None=@n10) {name=Particles, nodeColor=<rgba 1 0 0.0469 1>}
  ScalarParameter() {defaultValue=0.25, parameterName=Spec}
 ]
}
