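# Landscape material that previews erosion brush results in Landmass, with Height RT and Normal RT exposed as texture parameters.
# World XY maps into those textures using location, grid size and resolution from MPC_Erosion.
# Enable Erosion Preview blends in the derived world-space normal and scales a World Position Offset that moves vertices to the sampled height.
# A world-projected moss texture supplies surface shading; a stray "asdfasdf" vector parameter also adds to the offset.
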Material /Landmass/Landscape/ErosionBrush/Materials/Preview/ErosionBrush_Landscape_PreviewAsMID
graph {
 properties {bTangentSpaceNormal=false, bUseMaterialAttributes=true, pixelDepthOffsetMode=PDOM_Legacy, usedShadingModels=Default Lit}
 nodes [
  n1 = TextureSample(UVs=MaterialFunctionCall!(Scale=ScalarParameter() {defaultValue=1024, parameterName=Texture Scale}) {materialFunction=/Engine/Functions/Engine_MaterialFunctions01/Coordinates/WorldPosition-XY}.Result) {texture=/Engine/StarterContent/Textures/T_ground_Moss_D}
  Multiply(A=@n1.R, B=ScalarParameter() {defaultValue=0.5, parameterName=Specular})
  n2 = ComponentMask(None=WorldPosition().XYZ) {g=true, r=true}
  n3 = ComponentMask(None=CollectionParameter() {collection=/Landmass/Landscape/BlueprintBrushes/MPC/MPC_Erosion, parameterName=LandscapeLocation}) {g=true, r=true}
  n4 = Divide(A=CollectionParameter() {collection=&MPC_Erosion, parameterName=GridSize})
  n5 = ScalarParameter() {defaultValue=-1, parameterName=Shift}
  n6 = CollectionParameter() {collection=&MPC_Erosion, parameterName=GridSize}
  n7 = CollectionParameter() {collection=&MPC_Erosion, parameterName=ResX}
  n8 = CollectionParameter() {collection=&MPC_Erosion, parameterName=ResY}
  n9 = Divide(
   A=Subtract(A=Subtract(A=@n2, B=@n3), B=Multiply(A=@n4, B=@n5))
   B=Multiply(A=ComponentMask(None=@n6) {g=true, r=true}, B=AppendVector(A=@n7, B=@n8))
  )
  n10 = ConstantBiasScale(None=TextureSampleParameter2D(UVs=@n9) {parameterName=Normal RT, samplerType=SAMPLERTYPE_LinearColor, texture=/Engine/ArtTools/RenderToTexture/Textures/127grey}.RGB) {bias=-0.5, scale=2}
  NamedRerouteDeclaration(None=DeriveNormalZ(InXY=ComponentMask(None=@n10) {g=true, r=true})) {name=Erosion Normal, nodeColor=<rgba 0.6875 1 0 1>}
  n11 = OneMinus(None=@n1.A)
  n12 = VertexNormalWS()
  n13 = NamedRerouteUsage()
  n14 = ScalarParameter() {parameterName=Enable Erosion Preview}
  n15 = TextureSampleParameter2D(UVs=@n9) {
   parameterName=Height RT
   samplerType=SAMPLERTYPE_LinearColor
   texture=/Engine/Functions/Engine_MaterialFunctions02/PivotPainter2/Black_1x1_EXR_Texture
  }
  n16 = WorldPosition()
  n17 = ComponentMask(None=CollectionParameter() {collection=&MPC_Erosion, parameterName=LandscapeLocation}) {b=true}
  n18 = MaterialFunctionCall(Z=Subtract(A=@n15.R, B=Subtract(A=@n16.Z, B=@n17))) {materialFunction=&MakeFloat3}
  n19 = ScalarParameter() {parameterName=Enable Erosion Preview}
  SetMaterialAttributes!(
   MaterialAttributes=SetMaterialAttributes!(
    MaterialAttributes=SetMaterialAttributes!(MaterialAttributes, Base Color=@n1.RGB, Specular=@n1.RGB, Roughness=@n11) {attributeSetTypes [#1, #2, #3]}
    Normal=LinearInterpolate(A=@n12, B=@n13, Alpha=@n14)
   ) {attributeSetTypes [#4]}
   World Position Offset=Add(A=Multiply(A=@n18.Result, B=@n19), B=VectorParameter() {parameterName=asdfasdf}.RGB)
  ) {attributeSetTypes [#5]}
  TextureSample() {texture=/Engine/EngineMaterials/T_Default_MacroVariation}
 ]
}
